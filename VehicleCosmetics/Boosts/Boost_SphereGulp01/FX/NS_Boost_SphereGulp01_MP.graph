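# Niagara system for the SphereGulp01 vehicle boost cosmetic, trailing colored ribbons and tumbling balls behind the car.
# Blue, red and green ribbon emitters start at the single-boost offset, inherit fading vehicle movement and spawn faster in frontend preview.
# Two identical ball emitters spawn mesh and sprite particles at skeletal-mesh sockets, adding curl noise, drag and downward acceleration.
# Their paired collision modules swap on the preview flag, using analytical planes in the frontend.
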NiagaraSystem /VehicleCosmetics/Boosts/Boost_SphereGulp01/FX/NS_Boost_SphereGulp01_MP
dependencies {
 usedDataInterfaces [&NiagaraDataInterfaceCurve, &NiagaraDataInterfaceDebugDraw, &NiagaraDataInterfaceSkeletalMesh]
 usedDynamicInputs [&Vector2DFromFloat]
 usedModules [
  &self:SystemSpawnScript.SystemScriptSource.SystemScriptGraph.NiagaraNodeAssignment_0.SetVariables_#1
  &SystemState
 ]
}
emitters {
 AzulSingle {
  emitterUpdateScript [
   &EmitterState()
   &VC_FilteredSocketCount()
   &VC_RateQualityHandler(
    Epic Spawn Rate:NiagaraFloat=&MakeCustomFloatFromBool(Bool=$User.bIsFrontendPreview, True Float=60, False Float=30)
    High Spawn Rate:NiagaraFloat=&MakeCustomFloatFromBool(Bool=$User.bIsFrontendPreview, True Float=60, False Float=30)
    Low Spawn Rate:NiagaraFloat=&MakeCustomFloatFromBool(Bool=$User.bIsFrontendPreview, True Float=60, False Float=30)
    Medium Spawn Rate:NiagaraFloat=&MakeCustomFloatFromBool(Bool=$User.bIsFrontendPreview, True Float=60, False Float=30)
    Use Custom Rate?=true
   )
   &VC_SpawnRibbon_v1(SpawnRate=$Emitter.Spawn Rate)
  ]
  particleSpawnScript [
   &VC_PaintedVariantsOverride(Pink=<rgba 0.5938 0.1175 0.4206 1>, Purple=<rgba 0.4473 0.0002 0.849 1>, SkyBlue=<rgba 0.1997 0.7045 0.9583 1>) {enabled=false}
   &VC_InitializeRibbon_v1(
    Color=<rgba 0 0.2476 1 1>
    Custom Offset=true
    Lifetime=1
    Local Vector Offset=&AddVectorToPosition(
     Position=<-5 0 0>
     Vector=&AddVectorToPosition(Position=$User.DTCarOffset, Vector=Linked() {linkedVariable {name=User.SingleBoostPos, type=NiagaraPosition}})
    )
    Offset Amount=1
    Ribbon Width=16
   )
   &AddVelocity(Velocity=<-1000 0 0>) {module=AddVelocity001}
   /SetVariables(Particles.Linear Color2:LinearColor=<rgba 0 0.1965 5 1>) {isSetParameters=true, module=SetVariables_#2}
  ]
  particleUpdateScript [
   &ParticleState()
   &InheritSourceMovement(
    Applied Position Scale=&VectorFromFloat(Value=&FloatFromCurve(FloatCurve=DataInterface() {propertyValues {Curve {keys [
          {interpMode=RCIM_Cubic, tangentMode=RCTM_Break, time=0, value=1}
          {arriveTangent=-2.9025, interpMode=RCIM_Cubic, tangentMode=RCTM_Break, time=1, value=0}
         ]}}}))
    ApplyForce=false
    ApplyPosition=true
    ApplyPositionFalloff=false
   )
   /SetVariables(
    Particles.Curve_color:LinearColor=&MultiplyLinearColor(Linear Color=$Particles.Linear Color2, Scale Factor=&MakeVector4(X=2, Y=2, Z=2, W=1))
    Particles.Scale_RibbonWidth:NiagaraFloat=&Multiply_Float(
     A=$Particles.Initial.RibbonWidth
     B=&FloatFromCurve(FloatCurve=DataInterface() {propertyValues {Curve {keys [<at 0 0>,<at 0.15 1>,<at 0.85 1>,<at 1 0>]}}}, Scale Curve=3)
    )
   ) {isSetParameters=true, module=SetVariables_#3}
   &ScaleRibbonWidth(Ribbon Width Scale=&FloatFromCurve(FloatCurve=DataInterface() {propertyValues {Curve {keys [<at 0 0>,<at 0.15 1>,<at 0.85 1>,<at 1 0>]}}}, Scale Curve=1.5))
   &ScaleColor(
    Scale Alpha=&FloatFromCurve(FloatCurve=DataInterface() {propertyValues {Curve {keys [<at 0 1>,<at 0.85 1>,<at 1 0>]}}})
    Scale RGB=<0.7 0.7 0.7>
    ScaleA=true
    ScaleRGB=true
   ) {module=ScaleColor002}
   &SolveForcesAndVelocity(Clamp Velocity=false, Limit Acceleration=false)
   &VC_UpdateRibbon_v1()
  ]
  rendererClasses [&NiagaraRibbonRendererProperties]
  renderers [
   {rendererClass=&NiagaraRibbonRendererProperties}
   {rendererClass=&NiagaraRibbonRendererProperties, rendererIndex=1}
  ]
 }
 AzulSingle001 {
  emitterUpdateScript [
   &EmitterState()
   &VC_FilteredSocketCount()
   &VC_RateQualityHandler(
    Epic Spawn Rate:NiagaraFloat=&MakeCustomFloatFromBool(Bool=$User.bIsFrontendPreview, True Float=60, False Float=30)
    High Spawn Rate:NiagaraFloat=&MakeCustomFloatFromBool(Bool=$User.bIsFrontendPreview, True Float=60, False Float=30)
    Low Spawn Rate:NiagaraFloat=&MakeCustomFloatFromBool(Bool=$User.bIsFrontendPreview, True Float=60, False Float=30)
    Medium Spawn Rate:NiagaraFloat=&MakeCustomFloatFromBool(Bool=$User.bIsFrontendPreview, True Float=60, False Float=30)
    Use Custom Rate?=true
   )
   &VC_SpawnRibbon_v1(SpawnRate=$Emitter.Spawn Rate)
  ]
  particleSpawnScript [
   &VC_PaintedVariantsOverride(Pink=<rgba 0.5938 0.1175 0.4206 1>, Purple=<rgba 0.4473 0.0002 0.849 1>, SkyBlue=<rgba 0.1997 0.7045 0.9583 1>) {enabled=false}
   &VC_InitializeRibbon_v1(
    Color=<rgba 1 0 0.0306 1>
    Custom Offset=true
    Lifetime=1
    Local Vector Offset=&AddVectorToPosition(
     Position=<-5 0 0>
     Vector=&AddVectorToPosition(Position=$User.DTCarOffset, Vector=Linked() {linkedVariable {name=User.SingleBoostPos, type=NiagaraPosition}})
    )
    Offset Amount=1
    Ribbon Width=16
   )
   &AddVelocity(Velocity=<-1000 0 0>) {module=AddVelocity001}
   /SetVariables(Particles.Linear Color2:LinearColor=<rgba 1.2486 0 0.0305 1>) {isSetParameters=true, module=SetVariables_#2}
  ]
  particleUpdateScript [
   &ParticleState()
   &InheritSourceMovement(
    Applied Position Scale=&VectorFromFloat(Value=&FloatFromCurve(FloatCurve=DataInterface() {propertyValues {Curve {keys [
          {interpMode=RCIM_Cubic, tangentMode=RCTM_Break, time=0, value=1}
          {arriveTangent=-2.9025, interpMode=RCIM_Cubic, tangentMode=RCTM_Break, time=1, value=0}
         ]}}}))
    ApplyForce=false
    ApplyPosition=true
    ApplyPositionFalloff=false
   )
   /SetVariables(
    Particles.Curve_color:LinearColor=&MultiplyLinearColor(
     Linear Color=$Particles.Linear Color2
     Scale Factor=&MakeVector4(
      X=2
      Y=2
      Z=2
      W=&FloatFromCurve(FloatCurve=DataInterface() {propertyValues {Curve {keys [<at 0 1>,<at 0.4 1>,<at 1 0>]}}})
     )
    )
    Particles.Scale_RibbonWidth:NiagaraFloat=&Multiply_Float(
     A=$Particles.Initial.RibbonWidth
     B=&FloatFromCurve(FloatCurve=DataInterface() {propertyValues {Curve {keys [<at 0 0>,<at 0.15 1>,<at 0.85 1>,<at 1 0>]}}}, Scale Curve=3)
    )
   ) {isSetParameters=true, module=SetVariables_#3}
   &ScaleRibbonWidth(Ribbon Width Scale=&FloatFromCurve(FloatCurve=DataInterface() {propertyValues {Curve {keys [<at 0 0>,<at 0.15 1>,<at 0.85 1>,<at 1 0>]}}}, Scale Curve=1.5))
   &ScaleColor(
    Scale Alpha=&FloatFromCurve(FloatCurve=DataInterface() {propertyValues {Curve {keys [<at 0 1>,<at 0.4 1>,<at 0.85 1>,<at 1 0>]}}})
    ScaleA=true
    ScaleRGB=true
   ) {module=ScaleColor002}
   &SolveForcesAndVelocity(Clamp Velocity=false, Limit Acceleration=false)
   &VC_UpdateRibbon_v1()
  ]
  rendererClasses [&NiagaraRibbonRendererProperties]
  renderers [
   {rendererClass=&NiagaraRibbonRendererProperties}
   {rendererClass=&NiagaraRibbonRendererProperties, rendererIndex=1}
  ]
 }
 VerdeSingle {
  emitterUpdateScript [
   &EmitterState()
   &VC_FilteredSocketCount()
   &VC_RateQualityHandler(
    Epic Spawn Rate:NiagaraFloat=&MakeCustomFloatFromBool(Bool=$User.bIsFrontendPreview, True Float=60, False Float=30)
    High Spawn Rate:NiagaraFloat=&MakeCustomFloatFromBool(Bool=$User.bIsFrontendPreview, True Float=60, False Float=30)
    Low Spawn Rate:NiagaraFloat=&MakeCustomFloatFromBool(Bool=$User.bIsFrontendPreview, True Float=60, False Float=30)
    Medium Spawn Rate:NiagaraFloat=&MakeCustomFloatFromBool(Bool=$User.bIsFrontendPreview, True Float=60, False Float=30)
    Use Custom Rate?=true
   )
   &VC_SpawnRibbon_v1(SpawnRate=$Emitter.Spawn Rate)
  ]
  particleSpawnScript [
   &VC_PaintedVariantsOverride(Pink=<rgba 0.5938 0.1175 0.4206 1>, Purple=<rgba 0.4473 0.0002 0.849 1>, SkyBlue=<rgba 0.1997 0.7045 0.9583 1>) {enabled=false}
   &VC_InitializeRibbon_v1(
    Color=<rgba 0 1 0.2377 1>
    Custom Offset=true
    Lifetime=1
    Local Vector Offset=&AddVectorToPosition(
     Position=<-5 0 0>
     Vector=&AddVectorToPosition(Position=$User.DTCarOffset, Vector=Linked() {linkedVariable {name=User.SingleBoostPos, type=NiagaraPosition}})
    )
    Offset Amount=1
    Ribbon Width=16
   )
   &AddVelocity(Velocity=<-1000 0 0>) {module=AddVelocity001}
   /SetVariables(Particles.Linear Color2:LinearColor=<rgba 1.6663 5 0 1>) {isSetParameters=true, module=SetVariables_#2}
  ]
  particleUpdateScript [
   &ParticleState()
   &InheritSourceMovement(
    Applied Position Scale=&VectorFromFloat(Value=&FloatFromCurve(FloatCurve=DataInterface() {propertyValues {Curve {keys [
          {interpMode=RCIM_Cubic, tangentMode=RCTM_Break, time=0, value=1}
          {arriveTangent=-2.9025, interpMode=RCIM_Cubic, tangentMode=RCTM_Break, time=1, value=0}
         ]}}}))
    ApplyForce=false
    ApplyPosition=true
    ApplyPositionFalloff=false
   )
   /SetVariables(
    Particles.Curve_color:LinearColor=&MultiplyLinearColor(Linear Color=$Particles.Linear Color2, Scale Factor=&MakeVector4(X=2, Y=2, Z=2, W=1))
    Particles.Scale_RibbonWidth:NiagaraFloat=&Multiply_Float(
     A=$Particles.Initial.RibbonWidth
     B=&FloatFromCurve(FloatCurve=DataInterface() {propertyValues {Curve {keys [<at 0 0>,<at 0.15 1>,<at 0.85 1>,<at 1 0>]}}}, Scale Curve=3)
    )
   ) {isSetParameters=true}
   &ScaleRibbonWidth(Ribbon Width Scale=&FloatFromCurve(FloatCurve=DataInterface() {propertyValues {Curve {keys [<at 0 0>,<at 0.15 1>,<at 0.85 1>,<at 1 0>]}}}, Scale Curve=1.5))
   &ScaleColor(
    Scale Alpha=&FloatFromCurve(FloatCurve=DataInterface() {propertyValues {Curve {keys [<at 0 1>,<at 0.85 1>,<at 1 0>]}}})
    ScaleA=true
    ScaleRGB=true
   ) {module=ScaleColor002}
   &SolveForcesAndVelocity(Clamp Velocity=false, Limit Acceleration=false)
   &VC_UpdateRibbon_v1()
  ]
  rendererClasses [&NiagaraRibbonRendererProperties]
  renderers [
   {rendererClass=&NiagaraRibbonRendererProperties}
   {rendererClass=&NiagaraRibbonRendererProperties, rendererIndex=1}
  ]
 }
 balls {
  emitterUpdateScript [
   &EmitterState()
   &VC_FilteredSocketCount()
   &SpawnRate(SpawnRate=3, Use Spawn Probability=false)
  ]
  particleSpawnScript [
   &VC_PaintedVariantsOverride() {enabled=false}
   &InitializeParticle(
    Color=<rgba 0.5 0.5 0.5 1>
    Mesh Scale Mode=::NewEnumerator3
    Mesh Uniform Scale=0.05
    Sprite Rotation Mode=::NewEnumerator2
    UsePositionOffset=false
    Write Material Random=true
   )
   &PartitionParticles(Number Of Partitions=$Emitter.PortCount)
   &SkeletalMeshLocation(
    Bone / Socket Index=$Particles.PartitionParticles.Partition
    Bone / Socket Sampling Mode=::NewEnumerator5
    MaskSpawnGroup=false
    Mesh Sampling Type=::NewEnumerator6
    Sampled Position Offset=&AddVectorToPosition(
     Position=<100 0 0>
     Vector=&AddVectorToPosition(Position=$User.DTCarOffset, Vector=Linked() {linkedVariable {name=User.SingleBoostPos, type=NiagaraPosition}})
    )
    Skeletal Mesh=$System.SkeletalMesh
   )
   &AddVelocity(Velocity=&RandomRangeVector(Minimum=<-1000 -55 40>, Maximum=<-1000 55 80>))
   &ShapeLocation(MaskSpawnGroup=false, Override Seed=false, Sphere Radius=50, UseOwnerScale=true)
   &InitialMeshOrientation(Use Rotation Vector=false)
  ]
  particleUpdateScript [
   &ParticleState()
   &ScaleMeshSize(Scale Factor=&VectorFromFloat(Value=&FloatFromCurve(FloatCurve=DataInterface() {propertyValues {Curve {keys [<at 0.85 1>,<at 1 0>]}}})))
   &Collision(
    Advanced Aging Rate=5
    Age Colliding Particles=true
    Analytical Collision Plane Position 1:Vector3f=<0 0 45>
    Analytical Collision Plane Position 2:Vector3f=<0 0 45>
    CPU Collision Type=::NewEnumerator3
    Collision Enabled=HlslExpression() {hlslExpression="User.bIsFrontendPreview == true ? true : false"}
    Control Roll On Collision=true
    Friction=1.5
    Friction During a Bounce:NiagaraFloat=&RandomRangeFloat(Minimum=0.9, Maximum=1.2)
    KillOnCollision=false
    Module.Control Rest State Via Surface Normal=false
    Module.EnableMaxCollisionCount=false
    Module.Randomize Collision Normal Vector=false
    Restitution=&RandomRangeFloat(Minimum=1, Maximum=1.5)
    ScaleRateOfRotation=false
   ) {module=Collision001}
   &Collision(
    Advanced Aging Rate=5
    Age Colliding Particles=true
    Collision Enabled=HlslExpression() {hlslExpression="User.bIsFrontendPreview == true ? false : true"}
    Control Roll On Collision=true
    Friction=1.5
    Friction During a Bounce:NiagaraFloat=&RandomRangeFloat(Minimum=0.9, Maximum=1.2)
    KillOnCollision=false
    Module.Control Rest State Via Surface Normal=false
    Module.EnableMaxCollisionCount=false
    Module.Randomize Collision Normal Vector=false
    Restitution=&RandomRangeFloat(Minimum=0, Maximum=0.3)
    ScaleRateOfRotation=false
   ) {module=Collision002}
   &UpdateMeshOrientation(Rotation Vector=<3 0 0>)
   &AccelerationForce(Acceleration=&RandomRangeVector(Minimum=<0 0 -650>, Maximum=<0 0 -350>))
   &ScaleColor(
    Scale Alpha=&FloatFromCurve(FloatCurve=DataInterface() {propertyValues {Curve {keys [<at 0 1 RCIM_Cubic>,<at 0.5 0>]}}})
    ScaleA=true
    ScaleRGB=true
   )
   &CurlNoiseForce(
    Bias Noise=false
    Noise Frequency=&RandomRangeFloat(Minimum=25, Maximum=65)
    Noise Strength=1000
    Pan Noise=false
   )
   &Drag(Drag=&RandomRangeFloat(Minimum=0.5, Maximum=1.5), Use Linear Drag=true, Use Rotational Drag=true)
   &InheritSourceMovement(
    Applied Position Scale=&VectorFromFloat(Value=&FloatFromCurve(FloatCurve=DataInterface() {propertyValues {Curve {keys [
          {interpMode=RCIM_Cubic, tangentMode=RCTM_Break, time=0, value=1}
          {arriveTangent=-3.0618, interpMode=RCIM_Cubic, tangentMode=RCTM_Break, time=1, value=0}
         ]}}}))
    ApplyForce=false
    ApplyPosition=true
    ApplyPositionFalloff=false
   )
   &SolveForcesAndVelocity(Clamp Velocity=false, Limit Acceleration=false)
   &DynamicMaterialParameters(Module.Index0Float0Write=true, Module.Index0Float1Write=true, Module.Index0Float2Write=true, Module.Index0Float3Write=true)
  ]
  rendererClasses [&NiagaraMeshRendererProperties, &NiagaraSpriteRendererProperties]
  renderers [
   {rendererClass=&NiagaraMeshRendererProperties, rendererIndex=0}
   {rendererClass=&NiagaraSpriteRendererProperties, rendererIndex=1}
  ]
 }
 balls001 {
  emitterUpdateScript [
   &EmitterState()
   &VC_FilteredSocketCount()
   &SpawnRate(SpawnRate=3, Use Spawn Probability=false)
  ]
  particleSpawnScript [
   &VC_PaintedVariantsOverride() {enabled=false}
   &InitializeParticle(
    Color=<rgba 0.5 0.5 0.5 1>
    Mesh Scale Mode=::NewEnumerator3
    Mesh Uniform Scale=0.05
    Sprite Rotation Mode=::NewEnumerator2
    UsePositionOffset=false
    Write Material Random=true
   )
   &PartitionParticles(Number Of Partitions=$Emitter.PortCount)
   &SkeletalMeshLocation(
    Bone / Socket Index=$Particles.PartitionParticles.Partition
    Bone / Socket Sampling Mode=::NewEnumerator5
    MaskSpawnGroup=false
    Mesh Sampling Type=::NewEnumerator6
    Sampled Position Offset=&AddVectorToPosition(
     Position=<100 0 0>
     Vector=&AddVectorToPosition(Position=$User.DTCarOffset, Vector=Linked() {linkedVariable {name=User.SingleBoostPos, type=NiagaraPosition}})
    )
    Skeletal Mesh=$System.SkeletalMesh
   )
   &AddVelocity(Velocity=&RandomRangeVector(Minimum=<-1000 -55 40>, Maximum=<-1000 55 80>))
   &ShapeLocation(MaskSpawnGroup=false, Override Seed=false, Sphere Radius=50, UseOwnerScale=true)
   &InitialMeshOrientation(Use Rotation Vector=false)
  ]
  particleUpdateScript [
   &ParticleState()
   &ScaleMeshSize(Scale Factor=&VectorFromFloat(Value=&FloatFromCurve(FloatCurve=DataInterface() {propertyValues {Curve {keys [<at 0.85 1>,<at 1 0>]}}})))
   &Collision(
    Advanced Aging Rate=5
    Age Colliding Particles=true
    Analytical Collision Plane Position 1:Vector3f=<0 0 45>
    Analytical Collision Plane Position 2:Vector3f=<0 0 45>
    CPU Collision Type=::NewEnumerator3
    Collision Enabled=HlslExpression() {hlslExpression="User.bIsFrontendPreview == true ? true : false"}
    Control Roll On Collision=true
    Friction=1.5
    Friction During a Bounce:NiagaraFloat=&RandomRangeFloat(Minimum=0.9, Maximum=1.2)
    KillOnCollision=false
    Module.Control Rest State Via Surface Normal=false
    Module.EnableMaxCollisionCount=false
    Module.Randomize Collision Normal Vector=false
    Restitution=&RandomRangeFloat(Minimum=1, Maximum=1.5)
    ScaleRateOfRotation=false
   ) {module=Collision001}
   &Collision(
    Advanced Aging Rate=5
    Age Colliding Particles=true
    Collision Enabled=HlslExpression() {hlslExpression="User.bIsFrontendPreview == true ? false : true"}
    Control Roll On Collision=true
    Friction=1.5
    Friction During a Bounce:NiagaraFloat=&RandomRangeFloat(Minimum=0.9, Maximum=1.2)
    KillOnCollision=false
    Module.Control Rest State Via Surface Normal=false
    Module.EnableMaxCollisionCount=false
    Module.Randomize Collision Normal Vector=false
    Restitution=&RandomRangeFloat(Minimum=0, Maximum=0.3)
    ScaleRateOfRotation=false
   ) {module=Collision002}
   &UpdateMeshOrientation(Rotation Vector=<3 0 0>)
   &AccelerationForce(Acceleration=&RandomRangeVector(Minimum=<0 0 -650>, Maximum=<0 0 -350>))
   &ScaleColor(
    Scale Alpha=&FloatFromCurve(FloatCurve=DataInterface() {propertyValues {Curve {keys [<at 0 1 RCIM_Cubic>,<at 0.5 0>]}}})
    ScaleA=true
    ScaleRGB=true
   )
   &CurlNoiseForce(
    Bias Noise=false
    Noise Frequency=&RandomRangeFloat(Minimum=25, Maximum=65)
    Noise Strength=1000
    Pan Noise=false
   )
   &Drag(Drag=&RandomRangeFloat(Minimum=0.5, Maximum=1.5), Use Linear Drag=true, Use Rotational Drag=true)
   &InheritSourceMovement(
    Applied Position Scale=&VectorFromFloat(Value=&FloatFromCurve(FloatCurve=DataInterface() {propertyValues {Curve {keys [
          {interpMode=RCIM_Cubic, tangentMode=RCTM_Break, time=0, value=1}
          {arriveTangent=-3.0618, interpMode=RCIM_Cubic, tangentMode=RCTM_Break, time=1, value=0}
         ]}}}))
    ApplyForce=false
    ApplyPosition=true
    ApplyPositionFalloff=false
   )
   &SolveForcesAndVelocity(Clamp Velocity=false, Limit Acceleration=false)
   &DynamicMaterialParameters(Module.Index0Float0Write=true, Module.Index0Float1Write=true, Module.Index0Float2Write=true, Module.Index0Float3Write=true)
  ]
  rendererClasses [&NiagaraMeshRendererProperties, &NiagaraSpriteRendererProperties]
  renderers [
   {rendererClass=&NiagaraMeshRendererProperties, rendererIndex=0}
   {rendererClass=&NiagaraSpriteRendererProperties, rendererIndex=1}
  ]
 }
}
summary {
 systemName=NS_Boost_SphereGulp01_MP
 emitters [
  {emitterName=AzulSingle, rendererClasses [&NiagaraRibbonRendererProperties]}
  {emitterName=VerdeSingle, rendererClasses [&NiagaraRibbonRendererProperties]}
  {emitterName=AzulSingle001, rendererClasses [&NiagaraRibbonRendererProperties]}
  {emitterName=balls001, rendererClasses [&NiagaraMeshRendererProperties, &NiagaraSpriteRendererProperties]}
  {emitterName=balls, rendererClasses [&NiagaraMeshRendererProperties, &NiagaraSpriteRendererProperties]}
 ]
 userVariables [
  User.bIsFrontendPreview {type=NiagaraBool, defaultValue=true}
  User.bIsPainted {type=NiagaraBool}
  User.ColorVariantID {type=NiagaraInt32}
  User.DTCarOffset {type=NiagaraPosition, description="", defaultValue {}}
  User.Fresnel {type=NiagaraFloat, defaultValue=7, description=""}
  User.GameModeID {type=NiagaraInt32}
  User.Gap_Metallic/Spec {type=NiagaraFloat, defaultValue=0, description=""}
  User.Gap_Roughness {type=NiagaraFloat, defaultValue=0.75, description=""}
  User.PaintedColor {type=LinearColor, description="", defaultValue {struct=LinearColor, value=<rgba 1 1 1 1>}}
  User.PaintedColor_2 {type=LinearColor, description="", defaultValue {struct=LinearColor, value=<rgba 1 1 1 1>}}
  User.PaintedColor_3 {type=LinearColor, description="", defaultValue {struct=LinearColor, value=<rgba 1 1 1 1>}}
  User.PaintedScalar {type=NiagaraFloat, defaultValue=1, description=""}
  User.PortCount {type=NiagaraInt32, defaultValue=1}
  User.SingleBoostPos {type=NiagaraPosition, description="", defaultValue {}}
  User.Square_Metallic/Specular {type=NiagaraFloat, defaultValue=1, description=""}
  User.Square_Roughness {type=NiagaraFloat, defaultValue=0.35, description=""}
 ]
}
